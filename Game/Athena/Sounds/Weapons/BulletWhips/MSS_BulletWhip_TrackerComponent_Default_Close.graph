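# One-shot mono MetaSound for close bullet whip sounds, using attenuation and whiz-by concurrency settings.
# On play, it picks a random heavy whip wave, randomizes an upward pitch shift and gain, then plays it.
# A character affiliation patch selects between enemy and friendly whip modulation bus values, the friendly one scaled by FriendlyMultiplier.
# That mix value multiplies the random gain to set output level, and playback completion fires OnFinished.
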MetaSoundSource /Game/Athena/Sounds/Weapons/BulletWhips/MSS_BulletWhip_TrackerComponent_Default_Close
inputs [
 Character.Affiliation {type=Enum:FortAudioAffiliation}
 Character.Health {type=Float}
 Character.UniqueID {type=Int32}
 FriendlyMultiplier {type=Float}
 UE.Source.OnPlay {type=Trigger}
 WhipSound {type=WaveAsset:Array}
]
interfaces [Fort.Character:1.0, UE.OutputFormat.Mono:1.0, UE.Source.OneShot:1.0, UE.Source:1.1]
outputs [
 "UE.OutputFormat.Mono.Audio:0" {type=Audio}
 UE.Source.OneShot.OnFinished {type=Trigger}
]
pages [Default {
  input_defaults {
   Character.Affiliation=0
   Character.Health=0
   Character.UniqueID=0
   FriendlyMultiplier=1
   UE.Source.OnPlay=null
   WhipSound [
    &/bullet_whip_heavy_01
    &/bullet_whip_heavy_02
    &/bullet_whip_heavy_03
    &/bullet_whip_heavy_04
   ]
  }
  nodes [
   n1 = Modulation.GetModulatorValue() {also {defaults {Modulator=/Game/Sounds/Modulation/Buses/CB_Weapon_Whip_Enemy}}}
   n2 = UE.Multiply:Float(
    PrimaryOperand=^FriendlyMultiplier
    AdditionalOperands=Modulation.GetModulatorValue() {also {defaults {Modulator=/Game/Sounds/Modulation/Buses/CB_Weapon_Whip_Friendly}}}.Out
   )
   ArrayRandomGet_0 = Array.Random Get:WaveAsset:Array(Next=^UE.Source.OnPlay, In Array=^WhipSound)
   n3 = UE.RandomFloat(Next=@ArrayRandomGet_0.On Next) {also {defaults {Max=2, Min=0.8}}}
   n4 = UE.RandomFloat(Next=@n3.On Next) {also {defaults {Min=0.5}}}
   WavePlayer_0 = UE.Wave Player:Mono(Wave Asset=@ArrayRandomGet_0.Value, Pitch Shift=@n3.Value, Play=@n4.On Next) {also {version [1,1]}}
   Multiply_0 = UE.Multiply:Audio by Float(
    AdditionalOperands=UE.Multiply:Float(
     AdditionalOperands=&MSP_AffiliationMixFloat(Affiliation=^Character.Affiliation, Enemy=@n1.Out, Non-Enemy=@n2.Out).Out
     PrimaryOperand=@n4.Value
    ).Out
    PrimaryOperand=@WavePlayer_0.Out Mono
   )
  ]
  links [{dst="@UE.OutputFormat.Mono.Audio:0", src=Multiply_0.Out}, {dst="@UE.Source.OneShot.OnFinished", src=WavePlayer_0.On Finished}]
 }]
properties {
 assetImportData=&self:AssetImportData
 attenuationSettings=&/bullet_whip_Att.Bullet_Whip_Att
 sampleRate=48000
 soundClassObject=&Fort_WeaponRanged_Projectile_Effects
 concurrencySet [/Game/Sounds/Concurrency_Rules/Athena_BulletWhizBy_Conc]
}
