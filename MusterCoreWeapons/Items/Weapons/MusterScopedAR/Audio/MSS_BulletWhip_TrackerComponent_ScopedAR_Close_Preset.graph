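# MetaSound Source preset for the close bullet whip sound used by the Muster Scoped AR's tracker component.
# It inherits its graph from the default close bullet whip MetaSound and supplies four heavy whip waves as the WhipSound array.
# Inputs expose Fort character data and a FriendlyMultiplier to that parent graph.
# It is a mono one-shot routed to the ranged projectile effects sound class, with bullet whip attenuation and whiz-by concurrency.
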MetaSoundSource /MusterCoreWeapons/Items/Weapons/MusterScopedAR/Audio/MSS_BulletWhip_TrackerComponent_ScopedAR_Close_Preset
preset_parent=/Game/Athena/Sounds/Weapons/BulletWhips/MSS_BulletWhip_TrackerComponent_Default_Close
inputs [
 Character.Affiliation {type=Enum:FortAudioAffiliation}
 Character.Health {type=Float}
 Character.UniqueID {type=Int32}
 FriendlyMultiplier {type=Float}
 UE.Source.OnPlay {type=Trigger}
 WhipSound {type=WaveAsset:Array}
]
interfaces [Fort.Character:1.0, UE.OutputFormat.Mono:1.0, UE.Source.OneShot:1.0, UE.Source:1.1]
outputs [
 "UE.OutputFormat.Mono.Audio:0" {type=Audio}
 UE.Source.OneShot.OnFinished {type=Trigger}
]
pages [Default {input_defaults {
   Character.Affiliation=0
   Character.Health=0
   Character.UniqueID=0
   FriendlyMultiplier=1
   UE.Source.OnPlay=null
   WhipSound [
    /Game/Athena/Sounds/Weapons/BulletWhips/bullet_whip_heavy_01
    /Game/Athena/Sounds/Weapons/BulletWhips/bullet_whip_heavy_02
    /Game/Athena/Sounds/Weapons/BulletWhips/bullet_whip_heavy_03
    /Game/Athena/Sounds/Weapons/BulletWhips/bullet_whip_heavy_04
   ]
  }}]
properties {
 assetImportData=&self:AssetImportData
 attenuationSettings=/Game/Athena/Sounds/Weapons/BulletWhips/bullet_whip_Att.Bullet_Whip_Att
 sampleRate=48000
 soundClassObject=&Fort_WeaponRanged_Projectile_Effects
 concurrencySet [/Game/Sounds/Concurrency_Rules/Athena_BulletWhizBy_Conc]
}
