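# Sound cue for the scoped assault rifle's aim-down-sights-in foley in MusterCoreWeapons.
# A single wave player feeds a modulator at the root, which plays the one ADS-in sample with default modulation settings.
# The cue uses the weapon-foley sound class, borrows the weapon reload attenuation, and plays at reduced volume.
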SoundCue /MusterCoreWeapons/Items/Weapons/MusterScopedAR/Audio/Scoped_AR_ADS_IN_Cue
root=Modulator_0
nodes [
 {id=Modulator_0, children [WavePlayer_1]}
 {id=WavePlayer_1, properties {soundWaveAssetPtr=&/Scoped_AR_ADS_In_01}}
]
properties {
 attenuationSettings=&Fort_Weapon_Reload_Att
 maxDistance=2250
 soundClassObject=&Fort_WeaponFoley
 volumeMultiplier=0.376
}
