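# First-person fire sound cue for the Muster scoped assault rifle, routed through a first-person ranged-weapon fire sound class.
# Mixes four layers: randomized close shots, randomized mechanical sounds, randomized shell impacts that are briefly delayed and pitched down, and a sub sweetener.
# An IsADS parameter continuously modulates the mechanical, shell and sweetener volumes, raising them when aiming down sights.
# Gunfire concurrency and manual reverb send settings apply to the cue.
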SoundCue /MusterCoreWeapons/Items/Weapons/MusterScopedAR/Audio/Scoped_AR_Fire_1P_Cue
root=SoundClass_1
nodes [
 {id=Delay_0, children [ModulatorContinuous_2], properties {delayMax=0.18, delayMin=0.18}}
 {id=Delay_1, children [WavePlayer_2]}
 {id=Mixer_1, children [Modulator_0, Delay_0, ModulatorContinuous_0, ModulatorContinuous_1], properties {inputVolume [1.15,1,1,1]}}
 {
  id=ModulatorContinuous_0
  children [Random_4]
  properties {volumeModulationParams {default=0, maxOutput=0.8, minOutput=0.6, paramMode=MPM_Abs, parameterName=IsADS}}
 }
 {
  id=ModulatorContinuous_1
  children [Delay_1]
  properties {volumeModulationParams {default=0, maxOutput=0.8, paramMode=MPM_Abs, parameterName=IsADS}}
 }
 {
  id=ModulatorContinuous_2
  children [Modulator_2]
  properties {volumeModulationParams {default=0, maxOutput=0.8, minOutput=0.4, paramMode=MPM_Abs, parameterName=IsADS}}
 }
 {id=Modulator_0, children [Random_0], properties {pitchMax=1, pitchMin=1}}
 {id=Modulator_2, children [Random_3], properties {pitchMax=0.75, pitchMin=0.7}}
 {id=Modulator_3, children [Mixer_1], properties {pitchMax=1, pitchMin=1}}
 {
  id=Random_0
  children [WavePlayer_0, WavePlayer_1, WavePlayer_11, WavePlayer_12, WavePlayer_13, WavePlayer_14]
  properties {bShouldExcludeFromBranchCulling=true, weights [1,1,1,1,1,1]}
 }
 {
  id=Random_3
  children [WavePlayer_20, WavePlayer_21, WavePlayer_22, WavePlayer_23, WavePlayer_24, WavePlayer_25]
  properties {bShouldExcludeFromBranchCulling=true, weights [1,1,1,1,1,1]}
 }
 {
  id=Random_4
  children [WavePlayer_26, WavePlayer_27, WavePlayer_28, WavePlayer_29, WavePlayer_30, WavePlayer_31]
  properties {bShouldExcludeFromBranchCulling=true, weights [1,1,1,1,1,1]}
 }
 {id=SoundClass_1, children [Modulator_3], properties {soundClassOverride=&Fort_WeaponRanged_Fire_1P_PSM}}
 {id=WavePlayer_0, properties {soundWaveAssetPtr=&/Scoped_AR_Fire_1P_Close_01}}
 {id=WavePlayer_1, properties {soundWaveAssetPtr=&/Scoped_AR_Fire_1P_Close_02}}
 {id=WavePlayer_11, properties {soundWaveAssetPtr=&/Scoped_AR_Fire_1P_Close_03}}
 {id=WavePlayer_12, properties {soundWaveAssetPtr=&/Scoped_AR_Fire_1P_Close_04}}
 {id=WavePlayer_13, properties {soundWaveAssetPtr=&/Scoped_AR_Fire_1P_Close_05}}
 {id=WavePlayer_14, properties {soundWaveAssetPtr=&/Scoped_AR_Fire_1P_Close_06}}
 {id=WavePlayer_2, properties {soundWaveAssetPtr=&/Scoped_AR_Fire_ADS_SUB_Sweetener}}
 {id=WavePlayer_20, properties {soundWaveAssetPtr=&/Scoped_AR_ShellImpact_01}}
 {id=WavePlayer_21, properties {soundWaveAssetPtr=&/Scoped_AR_ShellImpact_02}}
 {id=WavePlayer_22, properties {soundWaveAssetPtr=&/Scoped_AR_ShellImpact_03}}
 {id=WavePlayer_23, properties {soundWaveAssetPtr=&/Scoped_AR_ShellImpact_04}}
 {id=WavePlayer_24, properties {soundWaveAssetPtr=&/Scoped_AR_ShellImpact_05}}
 {id=WavePlayer_25, properties {soundWaveAssetPtr=&/Scoped_AR_ShellImpact_06}}
 {id=WavePlayer_26, properties {soundWaveAssetPtr=&/Scoped_AR_Fire_1P_Mech_01}}
 {id=WavePlayer_27, properties {soundWaveAssetPtr=&/Scoped_AR_Fire_1P_Mech_02}}
 {id=WavePlayer_28, properties {soundWaveAssetPtr=&/Scoped_AR_Fire_1P_Mech_03}}
 {id=WavePlayer_29, properties {soundWaveAssetPtr=&/Scoped_AR_Fire_1P_Mech_04}}
 {id=WavePlayer_30, properties {soundWaveAssetPtr=&/Scoped_AR_Fire_1P_Mech_05}}
 {id=WavePlayer_31, properties {soundWaveAssetPtr=&/Scoped_AR_Fire_1P_Mech_06}}
]
properties {
 attenuationSettings=&Fort_Weapon_Projectile_FirstPerson_Athena_Att
 bDebug=true
 maxDistance=2097152
 soundClassObject=&Fort_WeaponRanged_Fire_1P_PSM
 volumeMultiplier=1.1
 attenuationOverrides {manualReverbSendLevel=-1.0165, reverbSendMethod=Manual}
 concurrencyOverrides {bLimitToOwner=true, bVolumeScaleCanRelease=true, maxCount=100, resolutionRule=StopOldest, volumeScale=0.9}
 concurrencySet [&Weapon_Gunfire_FirstPerson_Conc]
}
