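# Sound cue for the scoped assault rifle's gunfire, layering close, medium and distant shot variations.
# The close branch mixes random shot and mechanical waves under close attenuation and the 3P fire sound class.
# Medium and distant random waves each have their own attenuation and share the distant fire sound class.
# A root continuous modulator maps the WeaponID parameter to pitch across the whole mix.
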SoundCue /MusterCoreWeapons/Items/Weapons/MusterScopedAR/Audio/Scoped_AR_Fire_Cue
root=ModulatorContinuous_0
nodes [
 {
  id=Attenuation_118
  children [Modulator_4]
  properties {attenuationSettings=&Fort_Weapon_Projectile_Close_Athena_Att}
 }
 {
  id=Attenuation_119
  children [Modulator_0]
  properties {attenuationSettings=&Fort_Weapon_Projectile_Medium_Mono_Athena_Att}
 }
 {
  id=Attenuation_120
  children [Modulator_1]
  properties {attenuationSettings=&Fort_Weapon_Projectile_Distant_Mono_Athena_Att}
 }
 {id=Mixer_1, children [SoundClass_77, SoundClass_76], properties {inputVolume [1.3,1]}}
 {id=Mixer_3, children [Random_5,Random_4], properties {inputVolume [1.3,1.3]}}
 {id=Mixer_4, children [Attenuation_119, Attenuation_120], properties {inputVolume [1,0.5]}}
 {
  id=ModulatorContinuous_0
  children [Mixer_1]
  properties {pitchModulationParams {default=0, maxInput=4, maxOutput=1.5, minOutput=1, parameterName=WeaponID}}
 }
 {id=Modulator_0, children [Random_0], properties {pitchMax=1, pitchMin=1, volumeMax=1.5, volumeMin=1.7}}
 {id=Modulator_1, children [Random_1], properties {volumeMax=1.8, volumeMin=1.8}}
 {id=Modulator_4, children [Mixer_3], properties {pitchMin=1, volumeMax=0.9, volumeMin=1}}
 {
  id=Random_0
  children [WavePlayer_7, WavePlayer_8, WavePlayer_9, WavePlayer_21, WavePlayer_22, WavePlayer_23]
  properties {bShouldExcludeFromBranchCulling=true, weights [1,1,1,1,1,1]}
 }
 {
  id=Random_1
  children [WavePlayer_0,WavePlayer_2,WavePlayer_3,WavePlayer_4,WavePlayer_5,WavePlayer_6]
  properties {bShouldExcludeFromBranchCulling=true, weights [1,1,1,1,1,1]}
 }
 {id=Random_4, children [WavePlayer_33, WavePlayer_34, WavePlayer_35], properties {bShouldExcludeFromBranchCulling=true, weights [1,1,1]}}
 {
  id=Random_5
  children [WavePlayer_15, WavePlayer_16, WavePlayer_17, WavePlayer_18, WavePlayer_19, WavePlayer_20]
  properties {bShouldExcludeFromBranchCulling=true, weights [1,1,1,1,1,1]}
 }
 {id=SoundClass_76, children [Mixer_4], properties {soundClassOverride=&Fort_WeaponRanged_Fire_Distant}}
 {
  id=SoundClass_77
  children [Attenuation_118]
  properties {soundClassOverride=&Fort_WeaponRanged_Fire_3P_PSM}
 }
 {id=WavePlayer_0, properties {soundWaveAssetPtr=&/Scoped_AR_Fire_Distant_01}}
 {id=WavePlayer_15, properties {soundWaveAssetPtr=&/Scoped_AR_Fire_1P_Close_01}}
 {id=WavePlayer_16, properties {soundWaveAssetPtr=&/Scoped_AR_Fire_1P_Close_02}}
 {id=WavePlayer_17, properties {soundWaveAssetPtr=&/Scoped_AR_Fire_1P_Close_03}}
 {id=WavePlayer_18, properties {soundWaveAssetPtr=&/Scoped_AR_Fire_1P_Close_04}}
 {id=WavePlayer_19, properties {soundWaveAssetPtr=&/Scoped_AR_Fire_1P_Close_05}}
 {id=WavePlayer_2, properties {soundWaveAssetPtr=&/Scoped_AR_Fire_Distant_02}}
 {id=WavePlayer_20, properties {soundWaveAssetPtr=&/Scoped_AR_Fire_1P_Close_06}}
 {id=WavePlayer_21, properties {soundWaveAssetPtr=&/Scoped_AR_Fire_Medium_04}}
 {id=WavePlayer_22, properties {soundWaveAssetPtr=&/Scoped_AR_Fire_Medium_05}}
 {id=WavePlayer_23, properties {soundWaveAssetPtr=&/Scoped_AR_Fire_Medium_06}}
 {id=WavePlayer_3, properties {soundWaveAssetPtr=&/Scoped_AR_Fire_Distant_03}}
 {id=WavePlayer_33, properties {soundWaveAssetPtr=&/Scoped_AR_Fire_1P_Mech_01}}
 {id=WavePlayer_34, properties {soundWaveAssetPtr=&/Scoped_AR_Fire_1P_Mech_02}}
 {id=WavePlayer_35, properties {soundWaveAssetPtr=&/Scoped_AR_Fire_1P_Mech_03}}
 {id=WavePlayer_4, properties {soundWaveAssetPtr=&/Scoped_AR_Fire_Distant_04}}
 {id=WavePlayer_5, properties {soundWaveAssetPtr=&/Scoped_AR_Fire_Distant_05}}
 {id=WavePlayer_6, properties {soundWaveAssetPtr=&/Scoped_AR_Fire_Distant_06}}
 {id=WavePlayer_7, properties {soundWaveAssetPtr=&/Scoped_AR_Fire_Medium_01}}
 {id=WavePlayer_8, properties {soundWaveAssetPtr=&/Scoped_AR_Fire_Medium_02}}
 {id=WavePlayer_9, properties {soundWaveAssetPtr=&/Scoped_AR_Fire_Medium_03}}
]
properties {
 bDebug=true
 maxDistance=27500
 soundClassObject=&Fort_WeaponRanged_Fire_3P_PSM
 volumeMultiplier=0.45
 concurrencyOverrides {bLimitToOwner=true, maxCount=3, resolutionRule=StopOldest}
 concurrencySet [&Weapon_Gunfire_Local_Conc, &Weapon_Gunfire_Global_Conc]
}
