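# Sound cue for the first-person low-ammo fire sound of the Muster scoped AR.
# A single low-ammo fire wave feeds a continuous modulator that raises volume as the IsADS parameter increases, making it louder when aiming down sights.
# It routes to a degraded ranged-weapon low-ammo sound class, with a reduced overall volume and a concurrency override that blocks new instances at the limit.
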SoundCue /MusterCoreWeapons/Items/Weapons/MusterScopedAR/Audio/Scoped_AR_Low_Ammo_Cue
root=ModulatorContinuous_2
nodes [
 {
  id=ModulatorContinuous_2
  children [WavePlayer_0]
  properties {volumeModulationParams {default=0, maxOutput=1.5, minOutput=1, paramMode=MPM_Abs, parameterName=IsADS}}
 }
 {id=WavePlayer_0, properties {soundWaveAssetPtr=&/Scoped_AR_Fire_1P_Low_Ammo_01}}
]
properties {
 bOverrideConcurrency=true
 maxDistance=2097152
 soundClassObject=/Game/Sounds/FortSoundClassesAndModes/Fort_WeaponRanged_LowAmmo_Degraded_PSM
 volumeMultiplier=0.7067
 concurrencyOverrides {resolutionRule=PreventNew}
}
