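# Sound cue for the out-of-ammo sound of the Muster scoped AR weapon.
# It randomly picks one of two equally weighted wave variations, then a modulator slightly varies volume while keeping pitch fixed.
# It uses a ranged-weapon third-person fire sound class with a boosted volume multiplier.
# Overridden concurrency allows only one instance at a time and prevents new plays while one is active.
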SoundCue /MusterCoreWeapons/Items/Weapons/MusterScopedAR/Audio/Scoped_AR_OutofAmmo_Cue
root=Modulator_0
nodes [
 {id=Modulator_0, children [Random_0], properties {pitchMax=1, pitchMin=1, volumeMax=0.95, volumeMin=0.85}}
 {id=Random_0, children [WavePlayer_0,WavePlayer_1], properties {weights [1,1]}}
 {id=WavePlayer_0, properties {soundWaveAssetPtr=&/Scoped_AR_OutofAmmo_02}}
 {id=WavePlayer_1, properties {soundWaveAssetPtr=&/Scoped_AR_OutofAmmo_01}}
]
properties {
 bOverrideConcurrency=true
 maxDistance=2097152
 soundClassObject=&Fort_WeaponRanged_Fire_3P_PSM
 volumeMultiplier=1.5
 attenuationOverrides {falloffDistance=1600}
 concurrencyOverrides {maxCount=1, resolutionRule=PreventNew}
}
