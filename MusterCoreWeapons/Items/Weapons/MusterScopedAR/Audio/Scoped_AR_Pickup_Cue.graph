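# Sound cue for picking up the scoped assault rifle.
# A random node plays one of two pickup wave variations with equal weight, at reduced volume.
# It uses weapon pickup attenuation and the loot collect sound class, and references a shared pickup concurrency set alongside single-instance override values.
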SoundCue /MusterCoreWeapons/Items/Weapons/MusterScopedAR/Audio/Scoped_AR_Pickup_Cue
root=Random_0
nodes [
 {id=Random_0, children [WavePlayer_2,WavePlayer_3], properties {weights [1,1]}}
 {id=WavePlayer_2, properties {soundWaveAssetPtr=&/Scoped_AR_Pickup_01}}
 {id=WavePlayer_3, properties {soundWaveAssetPtr=&/Scoped_AR_Pickup_02}}
]
properties {
 attenuationSettings=&Fort_Weapon_Pickup_Att
 maxDistance=3500
 soundClassObject=&Fort_Loot_Collect
 volumeMultiplier=0.4615
 concurrencyOverrides {maxCount=1, resolutionRule=PreventNew}
 concurrencySet [&PickupMaximum_Oldest]
}
