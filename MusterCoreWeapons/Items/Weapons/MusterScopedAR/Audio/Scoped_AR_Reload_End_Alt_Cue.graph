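# Sound cue for the scoped assault rifle's reload-end foley, in an alternate variant.
# A random node picks between two reload-end waves at equal weight.
# It plays through the weapon foley sound class with shared weapon reload attenuation and slightly reduced volume.
# The cue overrides concurrency and has its debug flag switched on.
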SoundCue /MusterCoreWeapons/Items/Weapons/MusterScopedAR/Audio/Scoped_AR_Reload_End_Alt_Cue
root=Random_1
nodes [
 {id=Random_1, children [WavePlayer_0,WavePlayer_3], properties {weights [1,1]}}
 {id=WavePlayer_0, properties {soundWaveAssetPtr=&/Scoped_AR_Reload_End_01}}
 {id=WavePlayer_3, properties {soundWaveAssetPtr=&/Scoped_AR_Reload_End_02}}
]
properties {
 attenuationSettings=&Fort_Weapon_Reload_Att
 bDebug=true
 bOverrideConcurrency=true
 maxDistance=2250
 soundClassObject=&Fort_WeaponFoley
 volumeMultiplier=0.752
}
