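# Sound cue for the cartridge insert sound during a scoped assault rifle reload.
# A Random node with equal weights picks one of two cartridge wave variations.
# The cue uses weapon reload attenuation and the weapon foley sound class, with a raised volume multiplier.
# It overrides concurrency and has its debug flag turned on.
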SoundCue /MusterCoreWeapons/Items/Weapons/MusterScopedAR/Audio/Scoped_AR_Reload_Insert_Cue
root=Random_0
nodes [
 {id=Random_0, children [WavePlayer_1,WavePlayer_2], properties {weights [1,1]}}
 {id=WavePlayer_1, properties {soundWaveAssetPtr=&/Scoped_AR_Reload_Cartridge_01}}
 {id=WavePlayer_2, properties {soundWaveAssetPtr=&/Scoped_AR_Reload_Cartridge_02}}
]
properties {
 attenuationSettings=&Fort_Weapon_Reload_Att
 bDebug=true
 bOverrideConcurrency=true
 maxDistance=2250
 soundClassObject=&Fort_WeaponFoley
 volumeMultiplier=1.253
}
