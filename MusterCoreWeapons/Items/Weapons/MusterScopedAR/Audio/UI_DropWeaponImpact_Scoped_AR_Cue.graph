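# Sound cue for the landing impact of a dropped scoped assault rifle in the MusterCoreWeapons plugin.
# A Random node with equal weights picks one of two landed wave variations.
# It plays through the Fort_Loot_Collect sound class with boosted volume and its own short-range attenuation override.
# A concurrency override caps simultaneous instances and blocks new plays once the limit is reached.
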SoundCue /MusterCoreWeapons/Items/Weapons/MusterScopedAR/Audio/UI_DropWeaponImpact_Scoped_AR_Cue
root=Random_0
nodes [
 {id=Random_0, children [WavePlayer_0,WavePlayer_1], properties {weights [1,1]}}
 {id=WavePlayer_0, properties {soundWaveAssetPtr=&/Scoped_AR_Landed_01}}
 {id=WavePlayer_1, properties {soundWaveAssetPtr=&/Scoped_AR_Landed_02}}
]
properties {
 bOverrideAttenuation=true
 bOverrideConcurrency=true
 maxDistance=1000
 soundClassObject=&Fort_Loot_Collect
 volumeMultiplier=1.7
 attenuationOverrides {falloffDistance=1000, attenuationShapeExtents {x=0}}
 concurrencyOverrides {maxCount=3, resolutionRule=PreventNew}
}
