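; Weapon Blueprint for the Muster scoped assault rifle, extending a parent with scope handling.
; Targeting changes toggle scope material scalars for locally viewed pawns, unhide third-person visuals for simulated proxies on exit, and suppress shell effects in ADS.
; Audio parameters pass ADS state and an assigned weapon ID to differentiate fire sounds.
; Applying a wrap resets the glass material slot as a workaround and restores an active scope effect.
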(event EventBeginPlay
  (|CallParentFunction)
  ;[0 0.03429 0.3094] Audio Stuff
  ;;[0 0.3094 0.04008] Assign Random Weapon ID used to differentiate many instances of 3P Fire in-game
  (Class|BPCAudioID|AssignWeaponID (Variables|Default|GetBPC_Audio_Weapon_ID)))

(event EventOnSetTargeting
  (params (bNewIsTargeting :type Boolean))
  ;[0 0.03429 0.3094] Audio Stuff
  ;;[0.3094 0.005494 0] When ADS = True we add sweeteners to the 1P Fire Sound via this Param
  (Audio|Parameter|SetFloatParameter (AudioParameter|GetAudioParameterComponent) "IsADS" (Math|Conversions|ToFloat(Boolean) bNewIsTargeting) :!class /Script/AudioExtensions.AudioParameterControllerInterface))

(event EventOnPlayWeaponFireFX
  (params (bPersistentFire :type Boolean) (bSecondaryFire :type Boolean))
  (|CallParentFunction bPersistentFire bSecondaryFire)
  (CallFunction|ActivateShellsFX (not (stale EventOnSetTargeting bNewIsTargeting))) ; Turn off shells in ADS
  ;[0 0.03429 0.3094] Audio Stuff
  ;;[0.1455 0.3094 0] "WeaponID" is used to help differentiate 3P Fire when there are many instances of the same weapon in game firing at once.
  (Audio|Parameter|SetFloatParameter (AudioParameter|GetAudioParameterComponent) "WeaponID" (Math|Conversions|ToFloat(Integer) (Variables|/Game/Sounds/Fort_Audio_Blueprints/AudioIDComponent/BPC_Audio_ID.BPC_Audio_ID_C|GetWeaponID (Variables|Default|GetBPC_Audio_Weapon_ID))) :!class /Script/AudioExtensions.AudioParameterControllerInterface))

(event EventOnWrapApplied
  (params (WrapDefinition :type Athena%20Item%20Wrap%20Definition%20Object%20Reference))
  (Utilities|IsValid WrapDefinition
    (:Is%20Valid
      ;[1 1 1] Temporary workaround to "skip" applying Midas touch to scope glass
      (Rendering|Material|SetMaterial (Variables|Visual|GetWeaponMesh) 2 "/MusterCoreWeapons/Items/Weapons/MusterScopedAR/Materials/MI_Core_S23_Glass.MI_Core_S23_Glass")
      ;[1 1 1] If wrap was applied while we were ADS, restore ADS materials and stash new wrap materials
      (if (Variables|Default|GetMaterialEffectActive)
        (Default|TurnonScopeMaterialEffect)))))

(event Targeting|EventForceScopeFX)

(event Targeting|EventAbortScopeFX)

(fn GetScopeParameters
  (bind (_scopecomponent _depthoffieldvignetterange _weaponsightsoffset) (|CallParentFunction))
  (return (Variables|Default|GetEmptyScopeMesh) _depthoffieldvignetterange _weaponsightsoffset))

(fn HandleLocalTargetingChange
  (params (isTargeting :type Boolean))
  (Utilities|IsValid (Variables|Actor|GetInstigator)
    (:Is%20Valid
      (if isTargeting
        (bind _asfort_player_pawn (Utilities|Casting|CastToFortPlayerPawn (Variables|Actor|GetInstigator)))
        (if (Pawn|IsLocallyViewed _asfort_player_pawn)
          (Default|TurnonScopeMaterialEffect)
          (Audio|PlaySound2D 0))
        (else
          (bind _asfort_player_pawn_2 (Utilities|Casting|CastToFortPlayerPawn (Variables|Actor|GetInstigator)))
          (if (Pawn|IsLocallyViewed _asfort_player_pawn_2)
            (Default|TurnOffScopeMaterialEffect)
            (Audio|PlaySound2D 0)
            ;[1 1 1] Withthout this, the pawn is invisible for spectators if they died while this pawn was scoped in
            (if (Utilities|Enum|Equal(Enum) (Networking|GetLocalRole (Variables|Actor|GetInstigator)) "ROLE_SimulatedProxy")
              (CallFunction|UnhideThirdPersonStuff))))))))

(fn TurnOffScopeMaterialEffect
  (bind _weaponmesh (Variables|Visual|GetWeaponMesh))
  ;[1 1 1] Use this when wrap is disabled
  (Rendering|Material|SetScalarParameterValueonMaterials _weaponmesh "Scope_opacity" 0.0)
  (Rendering|Material|SetScalarParameterValueonMaterials _weaponmesh "WP_switch" 0.0)
  (Rendering|Material|SetScalarParameterValueonMaterials _weaponmesh "WP_switch_scopGlass" 0.0))

(fn TurnOnScopeMaterialEffect
  (bind _weaponmesh (Variables|Visual|GetWeaponMesh))
  ;[0.04 0.264 1] Use this when wrap is disabled
  (Rendering|Material|SetScalarParameterValueonMaterials _weaponmesh "Scope_opacity" 1.0)
  (Rendering|Material|SetScalarParameterValueonMaterials _weaponmesh "WP_switch" 1.0)
  (Rendering|Material|SetScalarParameterValueonMaterials _weaponmesh "WP_switch_scopGlass" 1.0))

(fn ConstructionScript
  (|CallParentFunction))
